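# Physics asset for the front belt mesh of the male medium GoldenGuardAntler cosmetic body, defining capsule bodies and joints for its dangling cloth and belt.
# From the pelvis, base_f_main leads into a three-link front skirt chain with limited swing and twist.
# A belt chain branches off dyn_f_main with tight limits past its freely rotating first joint.
# A separate two-bone left garment chain hangs from the pelvis with free angular motion.
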PhysicsAsset /BRCosmetics/Characters/Player/Male/Medium/Bodies/M_MED_GoldenGuardAntler/Meshes/M_MED_GoldenGuardAntler_FrontBelt_Physics
bodies {
 base_f_main [{shapeType=Capsule, length=0.505, radius=0.505, shapeName=base_f_main_capsule, rotation=<rot 0 -90 0>}]
 dyn_belt [{shapeType=Capsule, length=2, radius=2, center=<2 0 0>, rotation=<rot 90 0 0>}]
 dyn_belt_1 [{shapeType=Capsule, length=2, radius=2, center=<2 0 0>, rotation=<rot 90 0 0>}]
 dyn_belt_2 [{shapeType=Capsule, length=4, radius=3, center=<3 -1.6937 0>, rotation=<rot 90 0 0>}]
 dyn_belt_f [{
   shapeType=Capsule
   length=7.5659
   radius=4.2486
   shapeName=dyn_belt_f_capsule
   center=<0.6878 1.3921 2.3175>
   rotation=<rot -3.8153 -90 -10.0619>
  }]
 dyn_f_main [{shapeType=Capsule, length=49.1307, radius=6.0075, shapeName=dyn_garment_01_capsule, center=<6.4938 3.685 0>}]
 dyn_garment_l_00 [{shapeType=Capsule, length=22.0691, radius=2, shapeName=dyn_garment_l_00_capsule, center=<6.6568 -0.2894 -2.214>}]
 dyn_garment_l_01 [{shapeType=Capsule, length=22.1454, radius=3.4407, shapeName=dyn_garment_l_01_capsule, center=<2 3.7323 -3.0335>}]
 dyn_skirt_f_f [{shapeType=Capsule, length=45.8327, radius=6.7472, shapeName=dyn_garment_02_capsule, center=<2.0497 4.441 -1.5861>}]
 dyn_skirt_f_f_1 [{
   shapeType=Capsule
   length=50.1435
   radius=6.9505
   shapeName=dyn_garment_03_capsule
   center=<-4.7171 5.8176 -2.2229>
   rotation=<rot 0 -0.0002 -3.7151>
  }]
 pelvis [{
   shapeType=Capsule
   length=15.0027
   radius=14.4144
   shapeName=pelvis_capsule
   center=<1.4065 -1.4628 0.2287>
   rotation=<rot 3.7342 -90 -0.6972>
  }]
}
constraints [
 {bone1Name=dyn_skirt_f_f_1, bone2Name=dyn_skirt_f_f, swing1LimitDegrees=35, swing2LimitDegrees=30, twistLimitDegrees=20}
 {bone1Name=dyn_f_main, bone2Name=base_f_main, swing1LimitDegrees=30, swing2LimitDegrees=30, twistLimitDegrees=20}
 {bone1Name=dyn_skirt_f_f, bone2Name=dyn_f_main, swing1LimitDegrees=50, swing2LimitDegrees=30, twistLimitDegrees=20}
 {
  bone1Name=base_f_main
  bone2Name=pelvis
  swing1LimitDegrees=45
  swing1Motion=Free
  swing2LimitDegrees=45
  swing2Motion=Free
  twistLimitDegrees=45
  twistMotion=Free
 }
 {
  bone1Name=dyn_garment_l_01
  bone2Name=dyn_garment_l_00
  swing1LimitDegrees=45
  swing1Motion=Free
  swing2LimitDegrees=45
  swing2Motion=Free
  twistLimitDegrees=45
  twistMotion=Free
 }
 {
  bone1Name=dyn_garment_l_00
  bone2Name=pelvis
  swing1LimitDegrees=45
  swing1Motion=Free
  swing2LimitDegrees=45
  swing2Motion=Free
  twistLimitDegrees=45
  twistMotion=Free
 }
 {
  bone1Name=dyn_belt_f
  bone2Name=dyn_f_main
  swing1LimitDegrees=45
  swing1Motion=Free
  swing2LimitDegrees=45
  swing2Motion=Free
  twistLimitDegrees=45
  twistMotion=Free
 }
 {bone1Name=dyn_belt, bone2Name=dyn_belt_f, swing1LimitDegrees=5, swing2LimitDegrees=20, twistLimitDegrees=1}
 {bone1Name=dyn_belt_1, bone2Name=dyn_belt, swing1LimitDegrees=5, swing2LimitDegrees=20, twistLimitDegrees=1}
 {bone1Name=dyn_belt_2, bone2Name=dyn_belt_1, swing1LimitDegrees=5, swing2LimitDegrees=20, twistLimitDegrees=1}
]
