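# Physics asset for the front skirt of the male medium GoldenGuardAntler character body cosmetic.
# A chain of capsule bodies runs from a tiny base_f_main anchor through dyn_f_main and three dyn_skirt_f_f segments, each joint using limited swing and twist angles.
# Larger capsules cover the pelvis, thighs and calves as leg volumes.
# Their constraints, and the one tying the anchor to the pelvis, leave angular motion free.
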PhysicsAsset /BRCosmetics/Characters/Player/Male/Medium/Bodies/M_MED_GoldenGuardAntler/Meshes/M_MED_GoldenGuardAntler_FrontSkirt_Physics
bodies {
 base_f_main [{shapeType=Capsule, length=0.505, radius=0.505, shapeName=base_f_main_capsule, rotation=<rot 0 -90 0>}]
 calf_l [{shapeType=Capsule, length=36.9375, radius=12.875, center=<-24.5597 -2.8398 2.1719>, rotation=<rot 88.0594 -8.7171 -14.8289>}]
 calf_r [{shapeType=Capsule, length=38.4375, radius=11.375, center=<24.6324 2.8287 -0.0002>, rotation=<rot 88.5756 -12.1951 -17.4256>}]
 dyn_f_main [{shapeType=Capsule, length=16, radius=3, shapeName=dyn_garment_01_capsule, center=<2 -2.7 0>}]
 dyn_skirt_f_f [{shapeType=Capsule, length=12, radius=6, shapeName=dyn_garment_02_capsule, center=<3.3715 -4.2131 0>}]
 dyn_skirt_f_f_1 [{shapeType=Capsule, length=8.5, radius=6, shapeName=dyn_garment_03_capsule, center=<2 -4.6869 0>}]
 dyn_skirt_f_f_2 [{shapeType=Capsule, length=7.5, radius=4, shapeName=dyn_garment_04_capsule, center=<2 -3.1525 0>}]
 pelvis [{
   shapeType=Capsule
   length=15.0027
   radius=14.4144
   shapeName=pelvis_capsule
   center=<1.4065 -1.4628 0.2287>
   rotation=<rot 3.7342 -90 -0.6972>
  }]
 thigh_l [{
   shapeType=Capsule
   length=26.75
   radius=12.5
   shapeName=thigh_r_capsule
   center=<-17.2569 -2.0936 -1.2977>
   rotation=<rot 85.3118 179.8759 176.9445>
  }]
 thigh_r [{
   shapeType=Capsule
   length=27.75
   radius=12.5
   shapeName=thigh_r_capsule
   center=<15.4135 1.0948 0.028>
   rotation=<rot 86.7774 178.3456 -179.5829>
  }]
}
constraints [
 {
  bone1Name=calf_l
  bone2Name=thigh_l
  swing1LimitDegrees=45
  swing1Motion=Free
  swing2LimitDegrees=45
  swing2Motion=Free
  twistLimitDegrees=45
  twistMotion=Free
 }
 {
  bone1Name=calf_r
  bone2Name=thigh_r
  swing1LimitDegrees=45
  swing1Motion=Free
  swing2LimitDegrees=45
  swing2Motion=Free
  twistLimitDegrees=45
  twistMotion=Free
 }
 {bone1Name=dyn_skirt_f_f_1, bone2Name=dyn_skirt_f_f, swing1LimitDegrees=35, swing2LimitDegrees=30, twistLimitDegrees=20}
 {bone1Name=dyn_skirt_f_f_2, bone2Name=dyn_skirt_f_f_1, swing1LimitDegrees=25, swing2LimitDegrees=30, twistLimitDegrees=20}
 {bone1Name=dyn_f_main, bone2Name=base_f_main, swing1LimitDegrees=30, swing2LimitDegrees=30, twistLimitDegrees=20}
 {bone1Name=dyn_skirt_f_f, bone2Name=dyn_f_main, swing1LimitDegrees=50, swing2LimitDegrees=30, twistLimitDegrees=20}
 {
  bone1Name=thigh_l
  bone2Name=pelvis
  swing1LimitDegrees=45
  swing1Motion=Free
  swing2LimitDegrees=45
  swing2Motion=Free
  twistLimitDegrees=45
  twistMotion=Free
 }
 {
  bone1Name=thigh_r
  bone2Name=pelvis
  swing1LimitDegrees=45
  swing1Motion=Free
  swing2LimitDegrees=45
  swing2Motion=Free
  twistLimitDegrees=45
  twistMotion=Free
 }
 {
  bone1Name=base_f_main
  bone2Name=pelvis
  swing1LimitDegrees=45
  swing1Motion=Free
  swing2LimitDegrees=45
  swing2Motion=Free
  twistLimitDegrees=45
  twistMotion=Free
 }
]
