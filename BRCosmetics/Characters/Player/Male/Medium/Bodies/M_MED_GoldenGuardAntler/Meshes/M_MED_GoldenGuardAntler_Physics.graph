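# Physics asset for the GoldenGuardAntler male medium player body cosmetic, set up for its dynamic skirt.
# Capsules on pelvis, lower spine, thighs and calves form the body, and a huge root capsule below the origin approximates the ground.
# Seven four-bone skirt chains hang from the pelvis with limited swing and twist.
# Cross-constraints, mostly locked below the top row, tie each side's three chains into a panel, leaving the back chain separate.
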PhysicsAsset /BRCosmetics/Characters/Player/Male/Medium/Bodies/M_MED_GoldenGuardAntler/Meshes/M_MED_GoldenGuardAntler_Physics
bodies {
 calf_l [{shapeType=Capsule, length=47.2475, radius=9.9253, center=<-25.2513 -3.2238 -0.1407>, rotation=<rot 0 91.5239 90>}]
 calf_r [{shapeType=Capsule, length=47.1235, radius=9.9253, center=<25.077 3.2612 -0.1407>, rotation=<rot 0 90.7901 90>}]
 dyn_skirt_b_b [{shapeType=Capsule, length=19.3239, radius=5, shapeName=dyn_skirt_b_b_capsule, center=<7 -3 0>}]
 dyn_skirt_b_b_1 [{shapeType=Capsule, length=19.299, radius=5.3048, shapeName=dyn_skirt_b_b_1_capsule, center=<7 -4 0>}]
 dyn_skirt_b_b_2 [{shapeType=Capsule, length=19.3311, radius=5, shapeName=dyn_skirt_b_b_2_capsule, center=<7 -4.3354 0>}]
 dyn_skirt_b_b_3 [{shapeType=Capsule, length=31.4352, radius=5, shapeName=dyn_skirt_b_b_3_capsule, center=<14.7081 -6 0>}]
 dyn_skirt_l [{shapeType=Capsule, length=6, radius=5, shapeName=dyn_skirt_l_capsule, center=<7 -3 0>, rotation=<rot 90 0 0>}]
 dyn_skirt_l_1 [{shapeType=Capsule, length=6, radius=3, shapeName=dyn_skirt_l_1_capsule, center=<7 -4 0>, rotation=<rot 90 0 0>}]
 dyn_skirt_l_2 [{shapeType=Capsule, length=6, radius=3, shapeName=dyn_skirt_l_2_capsule, center=<7 -4 0>, rotation=<rot 90 0 0>}]
 dyn_skirt_l_3 [{shapeType=Capsule, length=8.4314, radius=5, shapeName=dyn_skirt_l_3_capsule, center=<8.0021 -4 0>, rotation=<rot 90 0 0>}]
 dyn_skirt_l_l [{shapeType=Capsule, length=6, radius=5, shapeName=dyn_skirt_l_l_capsule, center=<7 -3 0>, rotation=<rot 90 0 0>}]
 dyn_skirt_l_l_1 [{shapeType=Capsule, length=6, radius=3.2, shapeName=dyn_skirt_l_l_1_capsule, center=<7 -4 0>, rotation=<rot 90 0 0>}]
 dyn_skirt_l_l_2 [{shapeType=Capsule, length=6, radius=3.2, shapeName=dyn_skirt_l_l_2_capsule, center=<7 -4 0>, rotation=<rot 90 0 0>}]
 dyn_skirt_l_l_3 [{shapeType=Capsule, length=8.4314, radius=5, shapeName=dyn_skirt_l_3_capsule, center=<7.929 -4 0>, rotation=<rot 90 0 0>}]
 dyn_skirt_l_r [{shapeType=Capsule, length=6, radius=5, shapeName=dyn_skirt_l_r_capsule, center=<7 -3 0>, rotation=<rot 90 0 0>}]
 dyn_skirt_l_r_1 [{shapeType=Capsule, length=7.123, radius=3, shapeName=dyn_skirt_l_r_1_capsule, center=<7 -3.3619 0>, rotation=<rot 90 0 0>}]
 dyn_skirt_l_r_2 [{
   shapeType=Capsule
   length=12.3867
   radius=3
   shapeName=dyn_skirt_l_r_2_capsule
   center=<6.9899 -2.1167 2.3947>
   rotation=<rot 90 -3.0187 -3.1555>
  }]
 dyn_skirt_l_r_3 [{shapeType=Capsule, length=8.4314, radius=5.0099, shapeName=dyn_skirt_l_r_3_capsule, center=<8.0761 -4 0>, rotation=<rot 90 0 0>}]
 dyn_skirt_r [{shapeType=Capsule, length=6, radius=5, shapeName=dyn_skirt_r_capsule, center=<7 -4 0>, rotation=<rot 90 0 0>}]
 dyn_skirt_r_1 [{shapeType=Capsule, length=6, radius=3, shapeName=dyn_skirt_r_1_capsule, center=<6.9959 -4 0>, rotation=<rot 90 0 -0.1833>}]
 dyn_skirt_r_2 [{shapeType=Capsule, length=6, radius=3, shapeName=dyn_skirt_r_2_capsule, center=<7 -4 0>, rotation=<rot 90 0 0>}]
 dyn_skirt_r_3 [{shapeType=Capsule, length=8.4314, radius=5, shapeName=dyn_skirt_r_3_capsule, center=<7.9444 -4 0>, rotation=<rot 90 0 0>}]
 dyn_skirt_r_l [{shapeType=Capsule, length=6, radius=5, shapeName=dyn_skirt_r_3_capsule, center=<7 -4 0>, rotation=<rot 90 0 0>}]
 dyn_skirt_r_l_1 [{shapeType=Capsule, length=6, radius=3, shapeName=dyn_skirt_r_3_capsule, center=<7 -2.4939 -2.1121>, rotation=<rot 90 0 0>}]
 dyn_skirt_r_l_2 [{shapeType=Capsule, length=7.373, radius=3, shapeName=dyn_skirt_r_3_capsule, center=<6.3427 -0.9904 -2.5916>, rotation=<rot 90 0 0>}]
 dyn_skirt_r_l_3 [{shapeType=Capsule, length=8.4314, radius=5.0033, shapeName=dyn_skirt_r_3_capsule, center=<8.1524 -4 0>, rotation=<rot 90 0 0>}]
 dyn_skirt_r_r [{shapeType=Capsule, length=6, radius=5, shapeName=dyn_skirt_r_r_capsule, center=<7 -5.3562 0>, rotation=<rot 90 0 0>}]
 dyn_skirt_r_r_1 [{shapeType=Capsule, length=6, radius=3.2, shapeName=dyn_skirt_r_r_1_capsule, center=<7 -3.2362 0>, rotation=<rot 90 0 0>}]
 dyn_skirt_r_r_2 [{shapeType=Capsule, length=6.4566, radius=4, shapeName=dyn_skirt_r_r_2_capsule, center=<7 -5 0>, rotation=<rot 90 0 0>}]
 dyn_skirt_r_r_3 [{shapeType=Capsule, length=8.4314, radius=5, shapeName=dyn_skirt_r_r_3_capsule, center=<8.2951 -3.3998 0.8257>, rotation=<rot 90 0 0>}]
 pelvis [{shapeType=Capsule, length=6.25, radius=11.2297, center=<-1.844 -2.8954 -0.2594>}]
 root [{shapeType=Sphere, radius=1}, {shapeType=Capsule, length=15, radius=300, shapeName=root_capsule, center=<0 0 -299.5401>}]
 spine_01 [{shapeType=Capsule, length=6.6925, radius=10.4937, center=<3.96 0.599 0.0013>, rotation=<rot 0 -90 0>}]
 thigh_l [{shapeType=Capsule, length=33.3542, radius=8.5, center=<-23.565 0.1935 -0.8572>, rotation=<rot -0.0011 89.9908 -90.8708>}]
 thigh_r [{shapeType=Capsule, length=33.3542, radius=8.5, center=<23.338 0.1927 0.8702>, rotation=<rot -0.0011 89.9908 -90.8708>}]
}
constraints [
 {
  bone1Name=pelvis
  bone2Name=root
  swing1LimitDegrees=45
  swing1Motion=Free
  swing2LimitDegrees=45
  swing2Motion=Free
  twistLimitDegrees=45
  twistMotion=Free
 }
 {
  bone1Name=spine_01
  bone2Name=pelvis
  swing1LimitDegrees=45
  swing1Motion=Free
  swing2LimitDegrees=45
  swing2Motion=Free
  twistLimitDegrees=45
  twistMotion=Free
 }
 {
  bone1Name=thigh_l
  bone2Name=pelvis
  swing1LimitDegrees=45
  swing1Motion=Free
  swing2LimitDegrees=45
  swing2Motion=Free
  twistLimitDegrees=45
  twistMotion=Free
 }
 {
  bone1Name=calf_l
  bone2Name=thigh_l
  swing1LimitDegrees=45
  swing1Motion=Free
  swing2LimitDegrees=45
  swing2Motion=Free
  twistLimitDegrees=45
  twistMotion=Free
 }
 {
  bone1Name=thigh_r
  bone2Name=pelvis
  swing1LimitDegrees=45
  swing1Motion=Free
  swing2LimitDegrees=45
  swing2Motion=Free
  twistLimitDegrees=45
  twistMotion=Free
 }
 {
  bone1Name=calf_r
  bone2Name=thigh_r
  swing1LimitDegrees=45
  swing1Motion=Free
  swing2LimitDegrees=45
  swing2Motion=Free
  twistLimitDegrees=45
  twistMotion=Free
 }
 {bone1Name=dyn_skirt_b_b, bone2Name=pelvis, swing1LimitDegrees=15, swing2LimitDegrees=20, twistLimitDegrees=20}
 {bone1Name=dyn_skirt_b_b_1, bone2Name=dyn_skirt_b_b, swing1LimitDegrees=40, swing2LimitDegrees=30, twistLimitDegrees=1}
 {bone1Name=dyn_skirt_b_b_2, bone2Name=dyn_skirt_b_b_1, swing1LimitDegrees=40, swing2LimitDegrees=30, twistLimitDegrees=1}
 {bone1Name=dyn_skirt_b_b_3, bone2Name=dyn_skirt_b_b_2, swing1LimitDegrees=40, swing2LimitDegrees=30, twistLimitDegrees=1}
 {bone1Name=dyn_skirt_l, bone2Name=pelvis, swing1LimitDegrees=60, swing2LimitDegrees=10, twistLimitDegrees=20}
 {bone1Name=dyn_skirt_l_1, bone2Name=dyn_skirt_l, swing1LimitDegrees=45, swing2LimitDegrees=20, twistLimitDegrees=5}
 {bone1Name=dyn_skirt_l_2, bone2Name=dyn_skirt_l_1, swing1LimitDegrees=45, swing2LimitDegrees=20, twistLimitDegrees=5}
 {bone1Name=dyn_skirt_l_3, bone2Name=dyn_skirt_l_2, swing1LimitDegrees=45, swing2LimitDegrees=20, twistLimitDegrees=5}
 {bone1Name=dyn_skirt_l_l, bone2Name=pelvis, swing1LimitDegrees=60, swing2LimitDegrees=10, twistLimitDegrees=20}
 {bone1Name=dyn_skirt_l_l_1, bone2Name=dyn_skirt_l_l, swing1LimitDegrees=30, swing2LimitDegrees=10, twistLimitDegrees=5}
 {bone1Name=dyn_skirt_l_l_2, bone2Name=dyn_skirt_l_l_1, swing1LimitDegrees=30, swing2LimitDegrees=10, twistLimitDegrees=5}
 {bone1Name=dyn_skirt_l_l_3, bone2Name=dyn_skirt_l_l_2, swing1LimitDegrees=30, swing2LimitDegrees=30, twistLimitDegrees=10}
 {bone1Name=dyn_skirt_l_r, bone2Name=pelvis, swing1LimitDegrees=20, swing2LimitDegrees=10, twistLimitDegrees=20}
 {bone1Name=dyn_skirt_l_r_1, bone2Name=dyn_skirt_l_r, swing1LimitDegrees=45, swing2LimitDegrees=20, twistLimitDegrees=5}
 {bone1Name=dyn_skirt_l_r_2, bone2Name=dyn_skirt_l_r_1, swing1LimitDegrees=45, swing2LimitDegrees=20, twistLimitDegrees=5}
 {bone1Name=dyn_skirt_l_r_3, bone2Name=dyn_skirt_l_r_2, swing1LimitDegrees=45, swing2LimitDegrees=20, twistLimitDegrees=5}
 {bone1Name=dyn_skirt_r, bone2Name=pelvis, swing1LimitDegrees=30, swing2LimitDegrees=20, twistLimitDegrees=10}
 {bone1Name=dyn_skirt_r_1, bone2Name=dyn_skirt_r, swing1LimitDegrees=45, swing2LimitDegrees=20, twistLimitDegrees=5}
 {bone1Name=dyn_skirt_r_2, bone2Name=dyn_skirt_r_1, swing1LimitDegrees=45, swing2LimitDegrees=20, twistLimitDegrees=5}
 {bone1Name=dyn_skirt_r_3, bone2Name=dyn_skirt_r_2, swing1LimitDegrees=45, swing2LimitDegrees=20, twistLimitDegrees=5}
 {bone1Name=dyn_skirt_r_l, bone2Name=pelvis, swing1LimitDegrees=30, swing2LimitDegrees=20, twistLimitDegrees=10}
 {bone1Name=dyn_skirt_r_l_1, bone2Name=dyn_skirt_r_l, swing1LimitDegrees=45, swing2LimitDegrees=20, twistLimitDegrees=5}
 {bone1Name=dyn_skirt_r_l_2, bone2Name=dyn_skirt_r_l_1, swing1LimitDegrees=45, swing2LimitDegrees=20, twistLimitDegrees=5}
 {bone1Name=dyn_skirt_r_l_3, bone2Name=dyn_skirt_r_l_2, swing1LimitDegrees=45, swing2LimitDegrees=20, twistLimitDegrees=5}
 {bone1Name=dyn_skirt_r_r, bone2Name=pelvis, swing1LimitDegrees=30, swing2LimitDegrees=20, twistLimitDegrees=10}
 {bone1Name=dyn_skirt_r_r_1, bone2Name=dyn_skirt_r_r, swing1LimitDegrees=30, swing2LimitDegrees=10, twistLimitDegrees=5}
 {bone1Name=dyn_skirt_r_r_2, bone2Name=dyn_skirt_r_r_1, swing1LimitDegrees=30, swing2LimitDegrees=10, twistLimitDegrees=5}
 {bone1Name=dyn_skirt_r_r_3, bone2Name=dyn_skirt_r_r_2, swing1LimitDegrees=30, swing2LimitDegrees=10, twistLimitDegrees=5}
 {bone1Name=dyn_skirt_l_l, bone2Name=dyn_skirt_l, swing1LimitDegrees=5, swing2LimitDegrees=5, twistLimitDegrees=5}
 {bone1Name=dyn_skirt_l_r, bone2Name=dyn_skirt_l, swing1LimitDegrees=5, swing2LimitDegrees=5, twistLimitDegrees=5}
 {
  bone1Name=dyn_skirt_l_l_1
  bone2Name=dyn_skirt_l_1
  swing1LimitDegrees=45
  swing1Motion=Locked
  swing2LimitDegrees=45
  swing2Motion=Locked
  twistLimitDegrees=45
  twistMotion=Locked
 }
 {bone1Name=dyn_skirt_l_r_1, bone2Name=dyn_skirt_l_1, swing1LimitDegrees=5, swing2LimitDegrees=5, twistLimitDegrees=5}
 {
  bone1Name=dyn_skirt_l_l_2
  bone2Name=dyn_skirt_l_2
  swing1LimitDegrees=45
  swing1Motion=Locked
  swing2LimitDegrees=45
  swing2Motion=Locked
  twistLimitDegrees=45
  twistMotion=Locked
 }
 {
  bone1Name=dyn_skirt_l_r_2
  bone2Name=dyn_skirt_l_2
  swing1LimitDegrees=45
  swing1Motion=Locked
  swing2LimitDegrees=45
  swing2Motion=Locked
  twistLimitDegrees=45
  twistMotion=Locked
 }
 {
  bone1Name=dyn_skirt_l_r_3
  bone2Name=dyn_skirt_l_3
  swing1LimitDegrees=45
  swing1Motion=Locked
  swing2LimitDegrees=45
  swing2Motion=Locked
  twistLimitDegrees=45
  twistMotion=Locked
 }
 {
  bone1Name=dyn_skirt_l_l_3
  bone2Name=dyn_skirt_l_3
  swing1LimitDegrees=45
  swing1Motion=Locked
  swing2LimitDegrees=45
  swing2Motion=Locked
  twistLimitDegrees=45
  twistMotion=Locked
 }
 {bone1Name=dyn_skirt_r_l, bone2Name=dyn_skirt_r, swing1LimitDegrees=5, swing2LimitDegrees=5, twistLimitDegrees=5}
 {bone1Name=dyn_skirt_r_r, bone2Name=dyn_skirt_r, swing1LimitDegrees=5, swing2LimitDegrees=5, twistLimitDegrees=5}
 {
  bone1Name=dyn_skirt_r_l_1
  bone2Name=dyn_skirt_r_1
  swing1LimitDegrees=45
  swing1Motion=Locked
  swing2LimitDegrees=45
  swing2Motion=Locked
  twistLimitDegrees=45
  twistMotion=Locked
 }
 {
  bone1Name=dyn_skirt_r_r_1
  bone2Name=dyn_skirt_r_1
  swing1LimitDegrees=45
  swing1Motion=Locked
  swing2LimitDegrees=45
  swing2Motion=Locked
  twistLimitDegrees=45
  twistMotion=Locked
 }
 {
  bone1Name=dyn_skirt_r_l_2
  bone2Name=dyn_skirt_r_2
  swing1LimitDegrees=45
  swing1Motion=Locked
  swing2LimitDegrees=45
  swing2Motion=Locked
  twistLimitDegrees=45
  twistMotion=Locked
 }
 {
  bone1Name=dyn_skirt_r_r_2
  bone2Name=dyn_skirt_r_2
  swing1LimitDegrees=45
  swing1Motion=Locked
  swing2LimitDegrees=45
  swing2Motion=Locked
  twistLimitDegrees=45
  twistMotion=Locked
 }
 {
  bone1Name=dyn_skirt_r_l_3
  bone2Name=dyn_skirt_r_3
  swing1LimitDegrees=45
  swing1Motion=Locked
  swing2LimitDegrees=45
  swing2Motion=Locked
  twistLimitDegrees=45
  twistMotion=Locked
 }
 {
  bone1Name=dyn_skirt_r_r_3
  bone2Name=dyn_skirt_r_3
  swing1LimitDegrees=45
  swing1Motion=Locked
  swing2LimitDegrees=45
  swing2Motion=Locked
  twistLimitDegrees=45
  twistMotion=Locked
 }
]
