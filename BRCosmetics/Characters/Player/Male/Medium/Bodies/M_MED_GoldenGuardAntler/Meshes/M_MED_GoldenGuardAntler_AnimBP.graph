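# Animation Blueprint for the GoldenGuardAntler male medium body cosmetic, adding secondary motion to skirt, belts and garments.
# The pose copied from the parent mesh passes through a Control Rig, RigidBody and AnimDynamics simulations, then corrective bone offsets and constraints.
# Cached poses chain these stages, and crouching selects the final correction branch.
# The event graph derives movement-state alphas, directional movement, gravity override and an emote-dependent simulation space.
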AnimBlueprint /BRCosmetics/Characters/Player/Male/Medium/Bodies/M_MED_GoldenGuardAntler/Meshes/M_MED_GoldenGuardAntler_AnimBP
target_skeleton=&/M_MED_GoldenGuardAntler_Skeleton
graphs [
 AnimGraph {nodes [
   n1 = Component To Local(ComponentPose=Local To Component())
   n2 = Use cached pose 'Start Pose'()
   n3 = NOT Boolean(A=Get bIsBaseballSliding())
   Component To Local(ComponentPose=Local To Component(LocalPose=Blend Poses by bool(BlendPose_0=@n1, BlendPose_1=@n2, bActiveValue=@n3) {blendType=Linear}))
   Output Pose(Result=Use cached pose 'Post Dynamics Constraints'()) {layerGroup=None}
   n4 = Control Rig!(
    Source=Use cached pose 'Skirt Constraints'()
    Alpha
    is Crouching Alpha:double=Get is_crouching_alpha() -> (is_crouching_alpha:double)
    left right alpha:double=Get left_right_alpha() -> (left_right_alpha:double)
    is Sprinting Alpha:double=Get is_sprinting_alpha() -> (is_sprinting_alpha:double)
    is Baseball Sliding Alpha:double=Get is_baseball_sliding_alpha() -> (is_baseball_sliding_alpha:float)
    is Knee Sliding Alpha:double=Get is_knee_sliding_alpha() -> (is_knee_sliding_alpha:double)
    is Moving 2d Alpha:double=Get is_moving_2d_alpha() -> (is_moving_2d_alpha:float)
    is Jumping Alpha:double=Get is_jumping_alpha() -> (is_jumping_alpha:double)
    is Crouching:bool=Get bIsCrouching()
    Moving:bool=Get bIsMoving2D()
    is Parachuting Alpha:double=Get is Parachuting Alpha() -> (is Parachuting Alpha:double)
    IsEmote:bool=Get bIsPlayingEmote()
   ) -> (Pose) {
    alphaBoolBlend {blendOption=Linear}
    controlRigAssetReference {blueprintRigClass=&/M_MED_GoldenGuardAntler_CtrlRig.M_MED_GoldenGuardAntler_CtrlRig_C}
   }
   Save cached pose 'Control Rig'(Pose=@n4) {cacheName=Control Rig}
   n5 = "Transform (Modify) Bone - Bone: dyn_belt_sub_l"(ComponentPose=Local To Component(LocalPose=Use cached pose 'Rigid Body'()), Alpha=1, Translation="0,0,2.8", Rotation="0,0,0", Scale="1,1,1") {
    rotationMode=BMM_Additive
    rotationSpace=BCS_BoneSpace
    translationMode=BMM_Additive
    translationSpace=BCS_BoneSpace
    alphaBoolBlend {blendOption=Linear}
    alphaScaleBiasClamp {bInterpResult=true}
    boneToModify {boneName=dyn_belt_sub_l}
    translation {z=2.8}
   }
   n6 = "Transform (Modify) Bone - Bone: base_f_main"(ComponentPose=@n5, Alpha=1, Translation="0,0,0", Rotation="-0.016,12,0", Scale="1,1,1") {
    rotationMode=BMM_Additive
    rotationSpace=BCS_BoneSpace
    translationMode=BMM_Additive
    translationSpace=BCS_BoneSpace
    alphaBoolBlend {blendOption=Linear}
    alphaScaleBiasClamp {bInterpResult=true}
    boneToModify {boneName=base_f_main}
    rotation {pitch=-0.016, yaw=12}
   }
   n7 = "Transform (Modify) Bone - Bone: dyn_skirt_f_f"(ComponentPose=@n6, Alpha=1, Translation="0,-1.5,0", Rotation="0,-18,0", Scale="1,1,1") {
    rotationMode=BMM_Additive
    rotationSpace=BCS_BoneSpace
    translationMode=BMM_Additive
    translationSpace=BCS_BoneSpace
    alphaBoolBlend {blendOption=Linear}
    alphaScaleBiasClamp {bInterpResult=true}
    boneToModify {boneName=dyn_skirt_f_f}
    rotation {yaw=-18}
    translation {y=-1.5}
   }
   n8 = "Transform (Modify) Bone - Bone: dyn_belt_f"(ComponentPose=@n7, Alpha=1, Translation="0,-1.2,-0.294", Rotation="0,0,0", Scale="1,1,1") {
    rotationMode=BMM_Additive
    rotationSpace=BCS_BoneSpace
    translationMode=BMM_Additive
    translationSpace=BCS_BoneSpace
    alphaBoolBlend {blendOption=Linear}
    alphaScaleBiasClamp {bInterpResult=true}
    boneToModify {boneName=dyn_belt_f}
    translation {y=-1.2, z=-0.294}
   }
   n9 = "Transform (Modify) Bone - Bone: dyn_belt"(ComponentPose=@n8, Alpha=1, Translation="-0.021,-1,0.085", Rotation="0,-4.168,0", Scale="1,1,1") {
    rotationMode=BMM_Additive
    rotationSpace=BCS_BoneSpace
    translationMode=BMM_Additive
    translationSpace=BCS_BoneSpace
    alphaBoolBlend {blendOption=Linear}
    alphaScaleBiasClamp {bInterpResult=true}
    boneToModify {boneName=dyn_belt}
    rotation {yaw=-4.168}
    translation=<-0.021 -1 0.085>
   }
   n10 = "Transform (Modify) Bone - Bone: dyn_bag"(ComponentPose=@n9, Alpha=1, Translation="0.75,0,0", Rotation="4,-6,0", Scale="1,1,1") {
    rotationMode=BMM_Additive
    rotationSpace=BCS_BoneSpace
    translationMode=BMM_Additive
    translationSpace=BCS_BoneSpace
    alphaBoolBlend {blendOption=Linear}
    alphaScaleBiasClamp {bInterpResult=true}
    boneToModify {boneName=dyn_bag}
    rotation {pitch=4, yaw=-6}
    translation {x=0.75}
   }
   n11 = "Transform (Modify) Bone - Bone: dyn_skirt_l_r"(ComponentPose=@n10, Alpha=1, Translation="-1,0,0", Rotation="0,-5,0", Scale="1,1,1") {
    rotationMode=BMM_Additive
    rotationSpace=BCS_BoneSpace
    translationMode=BMM_Additive
    translationSpace=BCS_BoneSpace
    alphaBoolBlend {blendOption=Linear}
    alphaScaleBiasClamp {bInterpResult=true}
    boneToModify {boneName=dyn_skirt_l_r}
    rotation {yaw=-5}
    translation {x=-1}
   }
   n12 = "Transform (Modify) Bone - Bone: base_f_main"!(ComponentPose=@n11, bAlphaBoolEnabled:bool=True, Translation="0,-0.75,0", Rotation="0,-5,0", Scale="1,1,1") -> (Pose) {
    alphaInputType=Bool
    rotationMode=BMM_Additive
    rotationSpace=BCS_BoneSpace
    translationMode=BMM_Additive
    translationSpace=BCS_BoneSpace
    alphaBoolBlend {blendOption=Linear}
    alphaScaleBiasClamp {bInterpResult=true}
    boneToModify {boneName=base_f_main}
    rotation {yaw=-5}
    translation {y=-0.75}
   }
   Save cached pose 'Garment Constraints'(Pose=Component To Local(ComponentPose=@n12)) {cacheName=Garment Constraints}
   n13 = "Constraint - dyn_belt (PPP)"!(
    ComponentPose=Local To Component(LocalPose=Use cached pose 'Garment Constraints'())
    Alpha=1
    ConstraintWeights_0=0.3
    ConstraintWeights_1:float=0.2
    ConstraintWeights_2:float=0.5
   ) -> (Pose) {
    alphaBoolBlend {blendOption=Linear}
    boneToModify {boneName=dyn_belt}
    constraintSetup [{targetBone {boneName=dyn_f_main}}, {targetBone {boneName=dyn_skirt_f_f}}, {targetBone {boneName=dyn_garment_l_00}}]
    constraintWeights [1,1,1]
   }
   n14 = "Constraint - dyn_belt_1 (PP)"!(ComponentPose=@n13, Alpha=1, ConstraintWeights_0=0.8, ConstraintWeights_1:float=0.2) -> (Pose) {
    alphaBoolBlend {blendOption=Linear}
    boneToModify {boneName=dyn_belt_1}
    constraintSetup [{targetBone {boneName=dyn_skirt_f_f}}, {targetBone {boneName=dyn_garment_l_01}}]
    constraintWeights [1,1]
   }
   n15 = "Constraint - dyn_belt_2 (PPP)"!(ComponentPose=@n14, Alpha=1, ConstraintWeights_0=0.3, ConstraintWeights_1:float=0.2, ConstraintWeights_2:float=0.5) -> (Pose) {
    alphaBoolBlend {blendOption=Linear}
    boneToModify {boneName=dyn_belt_2}
    constraintSetup [{targetBone {boneName=dyn_skirt_f_f}}, {targetBone {boneName=dyn_garment_l_01}}, {targetBone {boneName=dyn_skirt_f_f_1}}]
    constraintWeights [1,1,1]
   }
   n16 = "Transform (Modify) Bone - Bone: dyn_belt_1"(ComponentPose=@n15, Alpha=1, Translation="0,-0.7,0", Rotation="0,0,0", Scale="1,1,1") {
    rotationMode=BMM_Additive
    rotationSpace=BCS_BoneSpace
    translationMode=BMM_Additive
    translationSpace=BCS_BoneSpace
    alphaBoolBlend {blendOption=Linear}
    boneToModify {boneName=dyn_belt_1}
    translation {y=-0.7}
   }
   n17 = Get forwards_backwards_alpha() -> (forwards_backwards_alpha:double)
   n18 = "Transform (Modify) Bone - Bone: dyn_belt_f"(
    ComponentPose=Local To Component(LocalPose=Use cached pose 'Garment Constraints'())
    Alpha=@n17
    Translation="0,-1,0"
    Rotation="0,-2,0"
    Scale="1,1,1"
   ) {
    rotationMode=BMM_Additive
    rotationSpace=BCS_BoneSpace
    translationMode=BMM_Additive
    translationSpace=BCS_BoneSpace
    alphaBoolBlend {blendOption=Linear}
    alphaScaleBiasClamp {bClampResult=true, bInterpResult=true}
    boneToModify {boneName=dyn_belt_f}
    rotation {yaw=-2}
    translation {y=-1}
   }
   n19 = "Transform (Modify) Bone - Bone: dyn_belt"(ComponentPose=@n18, Alpha=@n17, Translation="0,-2.5,0", Rotation="0,0,0", Scale="1,1,1") {
    rotationMode=BMM_Additive
    rotationSpace=BCS_BoneSpace
    translationMode=BMM_Additive
    translationSpace=BCS_BoneSpace
    alphaBoolBlend {blendOption=Linear}
    alphaScaleBiasClamp {bInterpResult=true, bMapRange=true}
    boneToModify {boneName=dyn_belt}
    translation {y=-2.5}
   }
   n20 = "Transform (Modify) Bone - Bone: dyn_belt"(ComponentPose=@n19, Alpha=@n17, Translation="0,-1,0", Rotation="0,0,0", Scale="1,1,1") {
    rotationMode=BMM_Additive
    rotationSpace=BCS_BoneSpace
    translationMode=BMM_Additive
    translationSpace=BCS_BoneSpace
    alphaBoolBlend {blendOption=Linear}
    alphaScaleBiasClamp {bInterpResult=true, bMapRange=true, inRange {max=-1}}
    boneToModify {boneName=dyn_belt}
    translation {y=-1}
   }
   n21 = Get left_right_alpha()
   n22 = "Transform (Modify) Bone - Bone: dyn_belt"(ComponentPose=@n20, Alpha=@n21, Translation="0,-1.7,0", Rotation="0,0,0", Scale="1,1,1") {
    rotationMode=BMM_Additive
    rotationSpace=BCS_BoneSpace
    translationMode=BMM_Additive
    translationSpace=BCS_BoneSpace
    alphaBoolBlend {blendOption=Linear}
    alphaScaleBiasClamp {bClampResult=true, bInterpResult=true}
    boneToModify {boneName=dyn_belt}
    translation {y=-1.7}
   }
   n23 = "Transform (Modify) Bone - Bone: dyn_belt"(ComponentPose=@n22, Alpha=@n21, Translation="0,-1.2,0", Rotation="0,0,0", Scale="1,1,1") {
    rotationMode=BMM_Additive
    rotationSpace=BCS_BoneSpace
    translationMode=BMM_Additive
    translationSpace=BCS_BoneSpace
    alphaBoolBlend {blendOption=Linear}
    alphaScaleBiasClamp {bInterpResult=true, bMapRange=true, inRange {max=-1}}
    boneToModify {boneName=dyn_belt}
    translation {y=-1.2}
   }
   n24 = "Transform (Modify) Bone - Bone: dyn_skirt_b_b_1"(ComponentPose=@n23, Alpha=Get is_skydiving_alpha() -> (is_skydiving_alpha:float), Translation="0,-3,0", Rotation="0,0,0", Scale="1,1,1") {
    rotationMode=BMM_Additive
    rotationSpace=BCS_BoneSpace
    translationMode=BMM_Additive
    translationSpace=BCS_BoneSpace
    alphaBoolBlend {blendOption=Linear}
    boneToModify {boneName=dyn_skirt_b_b_1}
    translation {y=-3}
   }
   Save cached pose 'Post Dynamics Constraints'(Pose=Blend Poses by bool(
     BlendPose_0=Component To Local(ComponentPose=@n16)
     BlendPose_1=Component To Local(ComponentPose=@n24)
     bActiveValue=Get Crouch moving() -> (Crouch moving:bool)
    ) {blendType=Linear}) {cacheName=Post Dynamics Constraints}
   n25 = RigidBody!(
    ComponentPose=Local To Component(LocalPose=Use cached pose 'Control Rig'())
    bAlphaBoolEnabled:bool
    OverrideWorldGravity:Vector=Get gravity_override() -> (gravity_override:Vector)
    ExternalForce
   ) -> (Pose) {
    alphaInputType=Bool
    bClampLinearTranslationLimitToRefPose=true
    bOverrideWorldGravity=true
    overridePhysicsAsset=&/M_MED_GoldenGuardAntler_FrontSkirt_Physics
    simulationSpace=BaseBoneSpace
    alphaBoolBlend {blendInTime=0.2, blendOption=Linear, blendOutTime=0.2}
    baseBoneRef {boneName=pelvis}
    componentAppliedLinearAccClamp=<500 500 500>
    componentLinearAccScale=<0.1 0.1 0.1>
    componentLinearVelScale=<0.1 0.1 0.1>
    overrideWorldGravity {z=-980}
    simSpaceSettings {
     maxAngularAcceleration=500
     maxAngularVelocity=3
     maxLinearAcceleration=400
     maxLinearVelocity=250
     velocityScaleZ=0
     worldAlpha=1
     externalLinearDragV=<0.1 0.1>
    }
   }
   n26 = Component To Local(ComponentPose=@n25)
   n27 = RigidBody(ComponentPose=Local To Component(LocalPose=Use cached pose 'Control Rig'()), OverrideWorldGravity=Get gravity_override()) {
    alphaInputType=Bool
    bClampLinearTranslationLimitToRefPose=true
    bOverrideWorldGravity=true
    overridePhysicsAsset=&/M_MED_GoldenGuardAntler_FrontSkirt_FE_Physics
    simulationSpace=BaseBoneSpace
    alphaBoolBlend {blendInTime=0.2, blendOption=Linear, blendOutTime=0.2}
    baseBoneRef {boneName=pelvis}
    componentAppliedLinearAccClamp=<500 500 500>
    componentLinearAccScale=<0.1 0.1 0.1>
    componentLinearVelScale=<0.1 0.1 0.1>
    overrideWorldGravity {z=-980}
    simSpaceSettings {
     maxAngularAcceleration=500
     maxAngularVelocity=3
     maxLinearAcceleration=400
     maxLinearVelocity=250
     velocityScaleZ=0
     worldAlpha=1
     externalLinearDragV=<0.1 0.1>
    }
   }
   n28 = Component To Local(ComponentPose=@n27)
   n29 = "Constraint - dyn_garment_l_00 (R)"(
    ComponentPose=Local To Component(LocalPose=Blend Poses by bool(BlendPose_0=@n28, BlendPose_1=@n26) {blendType=Linear})
    Alpha=1
    ConstraintWeights_0=1
   ) {
    alphaBoolBlend {blendOption=Linear}
    boneToModify {boneName=dyn_garment_l_00}
    constraintSetup [{transformType=Rotation, targetBone {boneName=dyn_f_main}}]
    constraintWeights [1]
   }
   n30 = "Constraint - dyn_garment_r_00 (R)"(ComponentPose=@n29, Alpha=1, ConstraintWeights_0=1) {
    alphaBoolBlend {blendOption=Linear}
    boneToModify {boneName=dyn_garment_r_00}
    constraintSetup [{transformType=Rotation, targetBone {boneName=dyn_f_main}}]
    constraintWeights [1]
   }
   n31 = "Constraint - dyn_belt_f (PP)"!(ComponentPose=@n30, Alpha=Get is_crouching_alpha(), ConstraintWeights_0=0.5, ConstraintWeights_1:float=0.5) -> (Pose) {
    alphaBoolBlend {blendOption=Linear}
    boneToModify {boneName=dyn_belt_f}
    constraintSetup [{targetBone {boneName=base_f_main}}, {targetBone {boneName=dyn_garment_l_00}}]
    constraintWeights [1,1]
   }
   Save cached pose 'Front Skirt Setup'(Pose=Component To Local(ComponentPose=@n31)) {cacheName=Front Skirt Setup}
   n32 = RigidBody(ComponentPose=Local To Component(LocalPose=Use cached pose 'Front Belt Setup'()), OverrideWorldGravity=Get gravity_override()) {
    alphaInputType=Bool
    bClampLinearTranslationLimitToRefPose=true
    bOverrideWorldGravity=true
    overridePhysicsAsset=&/M_MED_GoldenGuardAntler_Physics
    simulationSpace=BaseBoneSpace
    alphaBoolBlend {blendInTime=0.2, blendOption=Linear, blendOutTime=0.2}
    baseBoneRef {boneName=root}
    componentAppliedLinearAccClamp=<2500 2500 2500>
    componentLinearAccScale=<0.25 0.25 0.25>
    componentLinearVelScale=<0.25 0.25 0.25>
    overrideWorldGravity {z=-980}
    simSpaceSettings {worldAlpha=0.1}
   }
   n33 = Anim Dynamics - Bone: dyn_spine_belt_r(ComponentPose=@n32, Alpha=1, ExternalForce="0,0,0") {
    angularDampingOverride=0.9999
    angularSpringConstant=10000
    bAngularSpring=true
    bOverrideAngularDamping=true
    bUseGravityOverride=true
    gravityScale=0.35
    alphaBoolBlend {blendOption=Linear}
    boundBone {boneName=dyn_spine_belt_r}
    componentLinearAccScale=<1 1 1>
    componentLinearVelScale=<1 1 1>
    gravityOverride {z=-980}
    physicsBodyDefinitions [{
      boundBone {boneName=dyn_spine_belt_r}
      boxExtents=<30 30 30>
      constraintSetup {angularLimitsMax=<0 0 25>, angularLimitsMin=<0 0 -15>, angularTarget=<1 0 0>}
      localJointOffset=<2 0 0>
     }]
   }
   n34 = Anim Dynamics - Bone: dyn_spine_belt_l(ComponentPose=@n33, Alpha=1, ExternalForce="0,0,0") {
    angularDampingOverride=0.9999
    angularSpringConstant=10000
    bAngularSpring=true
    bOverrideAngularDamping=true
    bUseGravityOverride=true
    gravityScale=0.35
    alphaBoolBlend {blendOption=Linear}
    boundBone {boneName=dyn_spine_belt_l}
    componentLinearAccScale=<1 1 1>
    componentLinearVelScale=<1 1 1>
    gravityOverride {z=-980}
    physicsBodyDefinitions [{
      boundBone {boneName=dyn_spine_belt_l}
      boxExtents=<30 30 30>
      constraintSetup {angularLimitsMax=<0 0 25>, angularLimitsMin=<0 0 -15>, angularTarget=<1 0 0>}
      localJointOffset=<2 0 0>
     }]
   }
   Save cached pose 'Rigid Body'(Pose=Component To Local(ComponentPose=@n34)) {cacheName=Rigid Body}
   Save cached pose 'Start Pose'(Pose=Copy Pose From Mesh() {bUseAttachedParent=true}) {cacheName=Start Pose}
   Save cached pose 'Skirt Constraints'(Pose=Use cached pose 'Start Pose'()) {cacheName=Skirt Constraints}
   n35 = "Constraint - dyn_belt_1 (R)"!(ComponentPose=Local To Component(LocalPose=Use cached pose 'Front Skirt Setup'()), bAlphaBoolEnabled:bool=True, ConstraintWeights_0=1) -> (Pose) {
    alphaInputType=Bool
    alphaBoolBlend {blendInTime=0.1, blendOption=Linear, blendOutTime=0.1}
    boneToModify {boneName=dyn_belt_1}
    constraintSetup [{transformType=Rotation, targetBone {boneName=dyn_skirt_f_f}}]
    constraintWeights [1]
   }
   n36 = "Transform (Modify) Bone - Bone: dyn_belt"!(ComponentPose=@n35, bAlphaBoolEnabled:bool=True, Rotation="-0.000,-6.000,0.000") -> (Pose) {
    alphaInputType=Bool
    rotationMode=BMM_Additive
    rotationSpace=BCS_BoneSpace
    translationMode=BMM_Additive
    translationSpace=BCS_BoneSpace
    alphaBoolBlend {blendOption=Linear}
    alphaScaleBiasClamp {bInterpResult=true}
    boneToModify {boneName=dyn_belt}
    rotation {yaw=-6}
   }
   n37 = Anim Dynamics - Bone: dyn_belt!(SimulationSpace:AnimPhysSimSpaceType Enum=Component, ComponentPose=@n36, Alpha=1) -> (Pose) {
    angularDampingOverride=1
    angularSpringConstant=20000
    bAngularSpring=true
    bOverrideAngularDamping=true
    bUseGravityOverride=true
    gravityScale=0.35
    simulationSpace=RootRelative
    alphaBoolBlend {blendOption=Linear}
    boundBone {boneName=dyn_belt}
    componentLinearAccScale=<1 1 1>
    componentLinearVelScale=<1 1 1>
    gravityOverride {z=-980}
    physicsBodyDefinitions [{
      boundBone {boneName=dyn_belt}
      boxExtents=<20 30 30>
      constraintSetup {angularLimitsMax=<0 15 -2>, angularLimitsMin=<0 -15 -15>, angularTarget=<1 0 0>}
      localJointOffset=<2 0 0>
     }]
   }
   n38 = Anim Dynamics - Bone: dyn_belt_1!(SimulationSpace:AnimPhysSimSpaceType Enum, ComponentPose=@n37, Alpha=1) -> (Pose) {
    angularDampingOverride=1
    angularSpringConstant=20000
    bAngularSpring=true
    bOverrideAngularDamping=true
    bUseGravityOverride=true
    gravityScale=0.35
    simulationSpace=RootRelative
    alphaBoolBlend {blendOption=Linear}
    boundBone {boneName=dyn_belt_1}
    componentLinearAccScale=<1 1 1>
    componentLinearVelScale=<1 1 1>
    gravityOverride {z=-980}
    physicsBodyDefinitions [{
      boundBone {boneName=dyn_belt_1}
      boxExtents=<20 30 30>
      constraintSetup {angularLimitsMax=<0 15 -2>, angularLimitsMin=<0 -15 -15>, angularTarget=<1 0 0>}
      localJointOffset=<2 0 0>
     }]
   }
   n39 = Anim Dynamics - Bone: dyn_belt_2!(SimulationSpace:AnimPhysSimSpaceType Enum, ComponentPose=@n38, Alpha=1) -> (Pose) {
    angularDampingOverride=1
    angularSpringConstant=20000
    bAngularSpring=true
    bOverrideAngularDamping=true
    bUseGravityOverride=true
    gravityScale=0.35
    simulationSpace=RootRelative
    alphaBoolBlend {blendOption=Linear}
    boundBone {boneName=dyn_belt_2}
    componentLinearAccScale=<1 1 1>
    componentLinearVelScale=<1 1 1>
    gravityOverride {z=-980}
    physicsBodyDefinitions [{
      boundBone {boneName=dyn_belt_2}
      boxExtents=<20 30 30>
      constraintSetup {angularLimitsMax=<0 15 -2>, angularLimitsMin=<0 -15 -15>, angularTarget=<1 0 0>}
      localJointOffset=<2 0 0>
     }]
   }
   n40 = RigidBody(ComponentPose=Local To Component(LocalPose=Use cached pose 'Front Skirt Setup'()), OverrideWorldGravity=Get gravity_override()) {
    alphaInputType=Bool
    bClampLinearTranslationLimitToRefPose=true
    bOverrideWorldGravity=true
    overridePhysicsAsset=&/M_MED_GoldenGuardAntler_FrontBelt_Physics
    simulationSpace=BaseBoneSpace
    alphaBoolBlend {blendInTime=0.2, blendOption=Linear, blendOutTime=0.2}
    baseBoneRef {boneName=root}
    componentAppliedLinearAccClamp=<2500 2500 2500>
    componentLinearAccScale=<0.25 0.25 0.25>
    componentLinearVelScale=<0.25 0.25 0.25>
    overrideWorldGravity {z=-980}
    simSpaceSettings {worldAlpha=0.1}
   }
   Save cached pose 'Front Belt Setup'(Pose=Blend Poses by bool(BlendPose_0=Component To Local(ComponentPose=@n39), BlendPose_1=Component To Local(ComponentPose=@n40)) {blendType=Linear}) {cacheName=Front Belt Setup}
  ]}
 EventGraph {nodes [
   Event BlueprintUpdateAnimation()
   > n1 = Sequence!(execute) -> (then_0, then_1, then_2:Exec)
   > n2 = Gravity Override State Function DEPRECATED(CharacterPartAnimInstance=Self-Reference(), GameTimeInSeconds=GetGameTimeInSeconds(), GravityOverrideParams=Get Gravity Override Params())
   > Set gravity_override(gravity_override=@n2)
   > Set Moving Belt(Moving Belt:bool=OR Boolean(A=Get bIsMoving2D(), B=Get bIsFalling())) -> (Output_Get:bool)
   n3 = Get bIsPlayingEmote()
   > Set FE Emote(FE Emote:bool=AND Boolean(A=Get bIsInFrontEnd(), B=@n3)) -> (Output_Get:bool)
   > n4 = Branch(Condition=@n3)
   > Set Simulation Space(Simulation Space:AnimPhysSimSpaceType Enum) -> (Output_Get:AnimPhysSimSpaceType Enum)
   Set Simulation Space(execute=@n4.else, Simulation Space=World)
   Set is_crouching_alpha(execute=@n1.then_1, is_crouching_alpha:double="To Float (Boolean)"(InBool=Get bIsCrouching())) -> (Output_Get:double)
   > Set is_sprinting_alpha(is_sprinting_alpha:double="To Float (Boolean)"(InBool=Get bIsSprinting())) -> (Output_Get:double)
   > Set is_jumping_alpha(is_jumping_alpha:double="To Float (Boolean)"(InBool=Get bIsJumping())) -> (Output_Get:double)
   > Set is_knee_sliding_alpha(is_knee_sliding_alpha:double="To Float (Boolean)"(InBool=Get bIsKneeSliding())) -> (Output_Get:double)
   > Set is_baseball_sliding_alpha(is_baseball_sliding_alpha:float="To Float (Boolean)"(InBool=Get bIsBaseballSliding())) -> (Output_Get:float)
   > Set is_moving_2d_alpha(is_moving_2d_alpha:float="To Float (Boolean)"(InBool=Get bIsMoving2D())) -> (Output_Get:float)
   > Set is Parachuting Alpha(is Parachuting Alpha:double="To Float (Boolean)"(InBool=Get bIsParachuteOpened())) -> (Output_Get:double)
   n5 = Get bIsSkydiveDiveMode()
   n6 = Get bIsSkydiveFloating()
   > Set is_skydiving_alpha(is_skydiving_alpha:float="To Float (Boolean)"(InBool=OR Boolean(A=@n5, B=@n6))) -> (Output_Get:float)
   > Set Crouch moving(Crouch moving:bool=Get bIsCrouching()) -> (Output_Get:bool)
   > n7 = PlayerMovement(AnimInstance=Self-Reference(), XDotProductCoefficient=1, YDotProductCoefficient=1, ZDotProductCoefficient=1)
   > Set left_right_alpha(left_right_alpha:double=@n7) -> (Output_Get:double)
   > Set forwards_backwards_alpha(forwards_backwards_alpha:double=@n7.YMovement) -> (Output_Get:double)
   Reroute Node()
   n8 = Get bIsMoving2D()
   > n9 = Set Targeting not moving2d(Targeting not moving2d:bool=AND Boolean(A=NOT Boolean(A=@n8), B=Get bIsTargeting())) -> (Output_Get:bool)
   Reroute Node(InputPin=[@n10.OutputPin, @n13.then, @n9.then])
   n10 = Reroute Node()
   n11 = Get bIsMoving2D()
   n12 = Get bIsTargeting()
   n13 = Set Is not Moving 2D(Is not Moving 2D:bool=AND Boolean(A=NOT Boolean(A=@n11), B=NOT Boolean(A=@n12))) -> (Output_Get:bool)
   Get left_right_alpha()
   Get forwards_backwards_alpha()
  ]}
]
variables [
 Crouch moving {type=boolean}
 FE Emote {type=boolean}
 forwards_backwards_alpha {type=number}
 Gravity Override Params {type=GravityOverrideParamsStruct}
 gravity_override {type=Vector}
 Is not Moving 2D {type=boolean}
 is Parachuting Alpha {type=number}
 is_baseball_sliding_alpha {type=number}
 is_crouching_alpha {type=number}
 is_jumping_alpha {type=number}
 is_knee_sliding_alpha {type=number}
 is_moving_2d_alpha {type=number}
 is_skydiving_alpha {type=number}
 is_sprinting_alpha {type=number}
 left_right_alpha {type=number}
 Moving Belt {type=boolean}
 Override World Gravity for Front Skirt {type=Vector}
 Override World Gravity for Skirt {type=Vector}
 Simulation Space {type=AnimPhysSimSpaceType, default=RootRelative}
 Targeting not moving2d {type=boolean}
]
