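# Physics asset of colliders for the GoldenGuardAntler medium male body cosmetic, apparently for cape simulation to collide against, judging by its name.
# Capsules cover pelvis, spine, neck, clavicles and forearms, with a sphere on the head; leg and upper-arm bodies have no shapes.
# Five four-bone dyn_skirt chains hang from the pelvis with capsules, the back chain's third being tiny.
# Constraints follow the bone hierarchy with swing and twist left free.
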PhysicsAsset /BRCosmetics/Characters/Player/Male/Medium/Bodies/M_MED_GoldenGuardAntler/Meshes/M_MED_GoldenGuardAntler_CapeColliders
bodies {
 calf_l []
 calf_r []
 clavicle_l [{shapeType=Capsule, length=9.0192, radius=9.8871, center=<9.4266 -1.1806 -2.4977>, rotation=<rot 69.3297 1.4476 176.9462>}]
 clavicle_r [{shapeType=Capsule, length=9.0192, radius=9.8871, center=<-9.4266 1.1806 2.4977>, rotation=<rot 69.3297 1.4476 -3.0538>}]
 dyn_skirt_b_b [{
   shapeType=Capsule
   length=9.9074
   radius=7.3805
   shapeName=dyn_skirt_b_b_capsule
   center=<8.8154 3.979 -0.0968>
   rotation=<rot 0.2483 -90 -1.4392>
  }]
 dyn_skirt_b_b_1 [{
   shapeType=Capsule
   length=9.6114
   radius=7.6623
   shapeName=dyn_skirt_b_b_1_capsule
   center=<13.8563 4.1305 0.001>
   rotation=<rot 0.0281 -90 -0.0046>
  }]
 dyn_skirt_b_b_2 [{shapeType=Capsule, length=0.505, radius=0.505, shapeName=dyn_skirt_b_b_2_capsule, rotation=<rot 0 -90 0>}]
 dyn_skirt_b_b_3 [{
   shapeType=Capsule
   length=13.0708
   radius=7.9783
   shapeName=dyn_skirt_b_b_3_capsule
   center=<13.9621 3.5919 0.0033>
   rotation=<rot 0.0294 -90 -0.0068>
  }]
 dyn_skirt_l [{shapeType=Capsule, length=9, radius=2, shapeName=dyn_skirt_l_capsule, center=<2 0 0>, rotation=<rot 90 0 0>}]
 dyn_skirt_l_1 [{shapeType=Capsule, length=9, radius=3, shapeName=dyn_skirt_l_1_capsule, center=<2.9455 0 0>, rotation=<rot 90 0 0>}]
 dyn_skirt_l_2 [{shapeType=Capsule, length=9, radius=3, shapeName=dyn_skirt_l_2_capsule, center=<4.9136 0 0>, rotation=<rot 90 0 0>}]
 dyn_skirt_l_3 [{shapeType=Capsule, length=10.5918, radius=3.5791, shapeName=dyn_skirt_l_3_capsule, center=<8.0655 0 0>, rotation=<rot 90 0 0>}]
 dyn_skirt_l_l [{shapeType=Capsule, length=4.4209, radius=3.5791, shapeName=dyn_skirt_l_l_capsule, center=<2 0.4053 0>, rotation=<rot 90 0 0>}]
 dyn_skirt_l_l_1 [{shapeType=Capsule, length=6.8155, radius=3.5791, shapeName=dyn_skirt_l_l_1_capsule, center=<3.0197 0.4053 0>, rotation=<rot 90 0 0>}]
 dyn_skirt_l_l_2 [{shapeType=Capsule, length=7.3228, radius=3.5791, shapeName=dyn_skirt_l_l_2_capsule, center=<3.8183 0.4053 0>, rotation=<rot 90 0 0>}]
 dyn_skirt_l_l_3 [{shapeType=Capsule, length=9.9435, radius=5.3842, shapeName=dyn_skirt_r_r_3_capsule, center=<7.5764 1.3328 -0.6441>, rotation=<rot 90 0 0>}]
 dyn_skirt_r [{shapeType=Capsule, length=9, radius=2, shapeName=dyn_skirt_r_capsule, center=<2 0 0>, rotation=<rot 90 0 0>}]
 dyn_skirt_r_1 [{shapeType=Capsule, length=9, radius=3, shapeName=dyn_skirt_r_1_capsule, center=<2 0 0>, rotation=<rot 90 0 0>}]
 dyn_skirt_r_2 [{shapeType=Capsule, length=9, radius=3.5791, shapeName=dyn_skirt_r_2_capsule, center=<3.6931 0 0>, rotation=<rot 90 0 0>}]
 dyn_skirt_r_3 [{shapeType=Capsule, length=11.7486, radius=3.5791, shapeName=dyn_skirt_r_3_capsule, center=<7.5127 0 0>, rotation=<rot 90 0 0>}]
 dyn_skirt_r_r [{shapeType=Capsule, length=4.4209, radius=3.5791, shapeName=dyn_skirt_r_r_capsule, center=<2 0.4053 0>, rotation=<rot 90 0 0>}]
 dyn_skirt_r_r_1 [{shapeType=Capsule, length=6.8155, radius=3.5791, shapeName=dyn_skirt_r_r_1_capsule, center=<3.0197 0.4053 0>, rotation=<rot 90 0 0>}]
 dyn_skirt_r_r_2 [{shapeType=Capsule, length=7.3228, radius=3.5791, shapeName=dyn_skirt_r_r_2_capsule, center=<3.6525 0.4053 0>, rotation=<rot 90 0 0>}]
 dyn_skirt_r_r_3 [{shapeType=Capsule, length=13.7641, radius=5.3842, shapeName=dyn_skirt_r_r_3_capsule, center=<5.924 2.2893 0.6441>, rotation=<rot 90 0 0>}]
 foot_l []
 foot_r []
 head [{shapeType=Sphere, radius=17.3125, center=<11.028 -0.174 0>}]
 lowerarm_l [{shapeType=Capsule, length=6.4536, radius=6.7128, shapeName=lowerarm_l_capsule, center=<-1.8356 -5.8202 -0.3274>, rotation=<rot 90 0 15.2036>}]
 lowerarm_r [{shapeType=Capsule, length=6.4536, radius=6.7128, shapeName=lowerarm_l_capsule, center=<1.8356 5.8202 0.3274>, rotation=<rot 90 0 15.2036>}]
 neck_02 [{shapeType=Capsule, length=13.4375, radius=7.3175, center=<-0.2642 -0.3408 0>, rotation=<rot 90 -171.1193 -175.3978>}]
 pelvis [{shapeType=Capsule, length=8.6456, radius=14.5388, center=<-2.4887 0.1881 0>}]
 spine_01 [{shapeType=Capsule, length=7.9438, radius=13.5586, center=<-1.7965 0.1884 0>}]
 spine_02 [
  {shapeType=Capsule, length=7.3264, radius=13.6322, center=<-1.711 -0.1851 0.0001>}
  {shapeType=Capsule, length=29.3956, radius=11.5807, center=<2.5024 0.4992 3.9825>, rotation=<rot 90 0.0049 0.0049>}
  {shapeType=Capsule, length=29.875, radius=11.1013, center=<2.4704 0.4991 -3.9825>, rotation=<rot 90 -0.0396 -0.0396>}
 ]
 spine_03 [{shapeType=Capsule, length=8.2107, radius=13.7031, center=<0.1445 0.2396 0.0001>}]
 spine_04 [{shapeType=Capsule, length=9.4404, radius=14.1325, center=<2.766 0.7765 0.0002>}]
 spine_05 [
  {shapeType=Capsule, length=8.0226, radius=15.695, center=<-2.2795 1.8109 0>}
  {shapeType=Capsule, length=13.9518, radius=13.9425, center=<-0.2916 0.4452 0>}
 ]
 thigh_l []
 thigh_r []
 upperarm_l []
 upperarm_r []
}
constraints [
 {
  bone1Name=spine_01
  bone2Name=pelvis
  swing1LimitDegrees=45
  swing1Motion=Free
  swing2LimitDegrees=45
  swing2Motion=Free
  twistLimitDegrees=45
  twistMotion=Free
 }
 {
  bone1Name=spine_02
  bone2Name=spine_01
  swing1LimitDegrees=45
  swing1Motion=Free
  swing2LimitDegrees=45
  swing2Motion=Free
  twistLimitDegrees=45
  twistMotion=Free
 }
 {
  bone1Name=spine_03
  bone2Name=spine_02
  swing1LimitDegrees=45
  swing1Motion=Free
  swing2LimitDegrees=45
  swing2Motion=Free
  twistLimitDegrees=45
  twistMotion=Free
 }
 {
  bone1Name=spine_04
  bone2Name=spine_03
  swing1LimitDegrees=45
  swing1Motion=Free
  swing2LimitDegrees=45
  swing2Motion=Free
  twistLimitDegrees=45
  twistMotion=Free
 }
 {
  bone1Name=spine_05
  bone2Name=spine_04
  swing1LimitDegrees=45
  swing1Motion=Free
  swing2LimitDegrees=45
  swing2Motion=Free
  twistLimitDegrees=45
  twistMotion=Free
 }
 {
  bone1Name=neck_02
  bone2Name=spine_05
  swing1LimitDegrees=45
  swing1Motion=Free
  swing2LimitDegrees=45
  swing2Motion=Free
  twistLimitDegrees=45
  twistMotion=Free
 }
 {
  bone1Name=head
  bone2Name=neck_02
  swing1LimitDegrees=45
  swing1Motion=Free
  swing2LimitDegrees=45
  swing2Motion=Free
  twistLimitDegrees=45
  twistMotion=Free
 }
 {
  bone1Name=thigh_l
  bone2Name=pelvis
  swing1LimitDegrees=45
  swing1Motion=Free
  swing2LimitDegrees=45
  swing2Motion=Free
  twistLimitDegrees=45
  twistMotion=Free
 }
 {
  bone1Name=calf_l
  bone2Name=thigh_l
  swing1LimitDegrees=45
  swing1Motion=Free
  swing2LimitDegrees=45
  swing2Motion=Free
  twistLimitDegrees=45
  twistMotion=Free
 }
 {
  bone1Name=foot_l
  bone2Name=calf_l
  swing1LimitDegrees=45
  swing1Motion=Free
  swing2LimitDegrees=45
  swing2Motion=Free
  twistLimitDegrees=45
  twistMotion=Free
 }
 {
  bone1Name=clavicle_r
  bone2Name=spine_05
  swing1LimitDegrees=45
  swing1Motion=Free
  swing2LimitDegrees=45
  swing2Motion=Free
  twistLimitDegrees=45
  twistMotion=Free
 }
 {
  bone1Name=upperarm_r
  bone2Name=clavicle_r
  swing1LimitDegrees=45
  swing1Motion=Free
  swing2LimitDegrees=45
  swing2Motion=Free
  twistLimitDegrees=45
  twistMotion=Free
 }
 {
  bone1Name=lowerarm_r
  bone2Name=upperarm_r
  swing1LimitDegrees=45
  swing1Motion=Free
  swing2LimitDegrees=45
  swing2Motion=Free
  twistLimitDegrees=45
  twistMotion=Free
 }
 {
  bone1Name=foot_r
  bone2Name=calf_r
  swing1LimitDegrees=45
  swing1Motion=Free
  swing2LimitDegrees=45
  swing2Motion=Free
  twistLimitDegrees=45
  twistMotion=Free
 }
 {
  bone1Name=lowerarm_l
  bone2Name=upperarm_l
  swing1LimitDegrees=45
  swing1Motion=Free
  swing2LimitDegrees=45
  swing2Motion=Free
  twistLimitDegrees=45
  twistMotion=Free
 }
 {
  bone1Name=upperarm_l
  bone2Name=clavicle_l
  swing1LimitDegrees=45
  swing1Motion=Free
  swing2LimitDegrees=45
  swing2Motion=Free
  twistLimitDegrees=45
  twistMotion=Free
 }
 {
  bone1Name=clavicle_l
  bone2Name=spine_05
  swing1LimitDegrees=45
  swing1Motion=Free
  swing2LimitDegrees=45
  swing2Motion=Free
  twistLimitDegrees=45
  twistMotion=Free
 }
 {
  bone1Name=calf_r
  bone2Name=thigh_r
  swing1LimitDegrees=45
  swing1Motion=Free
  swing2LimitDegrees=45
  swing2Motion=Free
  twistLimitDegrees=45
  twistMotion=Free
 }
 {
  bone1Name=thigh_r
  bone2Name=pelvis
  swing1LimitDegrees=45
  swing1Motion=Free
  swing2LimitDegrees=45
  swing2Motion=Free
  twistLimitDegrees=45
  twistMotion=Free
 }
 {
  bone1Name=dyn_skirt_b_b
  bone2Name=pelvis
  swing1LimitDegrees=45
  swing1Motion=Free
  swing2LimitDegrees=45
  swing2Motion=Free
  twistLimitDegrees=45
  twistMotion=Free
 }
 {
  bone1Name=dyn_skirt_b_b_1
  bone2Name=dyn_skirt_b_b
  swing1LimitDegrees=45
  swing1Motion=Free
  swing2LimitDegrees=45
  swing2Motion=Free
  twistLimitDegrees=45
  twistMotion=Free
 }
 {
  bone1Name=dyn_skirt_b_b_2
  bone2Name=dyn_skirt_b_b_1
  swing1LimitDegrees=45
  swing1Motion=Free
  swing2LimitDegrees=45
  swing2Motion=Free
  twistLimitDegrees=45
  twistMotion=Free
 }
 {
  bone1Name=dyn_skirt_b_b_3
  bone2Name=dyn_skirt_b_b_2
  swing1LimitDegrees=45
  swing1Motion=Free
  swing2LimitDegrees=45
  swing2Motion=Free
  twistLimitDegrees=45
  twistMotion=Free
 }
 {
  bone1Name=dyn_skirt_r_r
  bone2Name=pelvis
  swing1LimitDegrees=45
  swing1Motion=Free
  swing2LimitDegrees=45
  swing2Motion=Free
  twistLimitDegrees=45
  twistMotion=Free
 }
 {
  bone1Name=dyn_skirt_r_r_1
  bone2Name=dyn_skirt_r_r
  swing1LimitDegrees=45
  swing1Motion=Free
  swing2LimitDegrees=45
  swing2Motion=Free
  twistLimitDegrees=45
  twistMotion=Free
 }
 {
  bone1Name=dyn_skirt_r_r_2
  bone2Name=dyn_skirt_r_r_1
  swing1LimitDegrees=45
  swing1Motion=Free
  swing2LimitDegrees=45
  swing2Motion=Free
  twistLimitDegrees=45
  twistMotion=Free
 }
 {
  bone1Name=dyn_skirt_r_r_3
  bone2Name=dyn_skirt_r_r_2
  swing1LimitDegrees=45
  swing1Motion=Free
  swing2LimitDegrees=45
  swing2Motion=Free
  twistLimitDegrees=45
  twistMotion=Free
 }
 {
  bone1Name=dyn_skirt_l_l
  bone2Name=pelvis
  swing1LimitDegrees=45
  swing1Motion=Free
  swing2LimitDegrees=45
  swing2Motion=Free
  twistLimitDegrees=45
  twistMotion=Free
 }
 {
  bone1Name=dyn_skirt_l_l_1
  bone2Name=dyn_skirt_l_l
  swing1LimitDegrees=45
  swing1Motion=Free
  swing2LimitDegrees=45
  swing2Motion=Free
  twistLimitDegrees=45
  twistMotion=Free
 }
 {
  bone1Name=dyn_skirt_l_l_2
  bone2Name=dyn_skirt_l_l_1
  swing1LimitDegrees=45
  swing1Motion=Free
  swing2LimitDegrees=45
  swing2Motion=Free
  twistLimitDegrees=45
  twistMotion=Free
 }
 {
  bone1Name=dyn_skirt_l_l_3
  bone2Name=dyn_skirt_l_l_2
  swing1LimitDegrees=45
  swing1Motion=Free
  swing2LimitDegrees=45
  swing2Motion=Free
  twistLimitDegrees=45
  twistMotion=Free
 }
 {
  bone1Name=dyn_skirt_l_1
  bone2Name=dyn_skirt_l
  swing1LimitDegrees=45
  swing1Motion=Free
  swing2LimitDegrees=45
  swing2Motion=Free
  twistLimitDegrees=45
  twistMotion=Free
 }
 {
  bone1Name=dyn_skirt_l_2
  bone2Name=dyn_skirt_l_1
  swing1LimitDegrees=45
  swing1Motion=Free
  swing2LimitDegrees=45
  swing2Motion=Free
  twistLimitDegrees=45
  twistMotion=Free
 }
 {
  bone1Name=dyn_skirt_l_3
  bone2Name=dyn_skirt_l_2
  swing1LimitDegrees=45
  swing1Motion=Free
  swing2LimitDegrees=45
  swing2Motion=Free
  twistLimitDegrees=45
  twistMotion=Free
 }
 {
  bone1Name=dyn_skirt_r_1
  bone2Name=dyn_skirt_r
  swing1LimitDegrees=45
  swing1Motion=Free
  swing2LimitDegrees=45
  swing2Motion=Free
  twistLimitDegrees=45
  twistMotion=Free
 }
 {
  bone1Name=dyn_skirt_r_2
  bone2Name=dyn_skirt_r_1
  swing1LimitDegrees=45
  swing1Motion=Free
  swing2LimitDegrees=45
  swing2Motion=Free
  twistLimitDegrees=45
  twistMotion=Free
 }
 {
  bone1Name=dyn_skirt_r_3
  bone2Name=dyn_skirt_r_2
  swing1LimitDegrees=45
  swing1Motion=Free
  swing2LimitDegrees=45
  swing2Motion=Free
  twistLimitDegrees=45
  twistMotion=Free
 }
 {
  bone1Name=dyn_skirt_l
  bone2Name=pelvis
  swing1LimitDegrees=45
  swing1Motion=Free
  swing2LimitDegrees=45
  swing2Motion=Free
  twistLimitDegrees=45
  twistMotion=Free
 }
 {
  bone1Name=dyn_skirt_r
  bone2Name=pelvis
  swing1LimitDegrees=45
  swing1Motion=Free
  swing2LimitDegrees=45
  swing2Motion=Free
  twistLimitDegrees=45
  twistMotion=Free
 }
]
